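; PCG example graph that places assets on a grid and builds frame data around them from reusable PCGPrimitives subgraphs.
; The placed grid feeds a bound footprint trace, which is expanded by the Frame_padding parameter.
; That expanded area branches into segment extraction with rule assignment, and a polyline branch that combines control points to compute a bounds Size attribute.
; Neither branch connects to a graph output, and the position parameter is unused.
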(graph Asset_grid_with_Frame
  (params (Assets :type (Array (SoftObject /Script/CoreUObject.Object))) (position :type Vector) (Frame_padding :type Float :default 3e+02))
  (bind PlaceOnGrid (Subgraph :Graph /PCGPrimitives/Primitives/Place/Place_Assets_OnAGrid.Place_Assets_OnAGrid :Assets Assets))
  (bind Subgraph_1 (Subgraph :Graph /PCGPrimitives/Primitives/Trace/Trace_Bound_Footprint.Trace_Bound_Footprint)) ; Retrieving placed assets to create a frame around them 
  (bind Subgraph_3 (Subgraph :Graph /PCGPrimitives/Primitives/Transform/Transform_Areas_Expand.Transform_Areas_Expand :Expand Frame_padding))
  (bind FilterDataByType_2 (Filter%20Data%20By%20Type :TargetType (struct :Ids [(struct :Struct /Script/PCG.PCGDataTypeInfoPolyline)])))
  (bind GetSplineControlPoints_0 (Get%20Spline%20Control%20Points))
  (bind CombinePoints_0 (Combine%20Points))
  (bind AttributeMathsOp_0 (Attribute%20Maths%20Op :InputSource1 $BoundsMax :InputSource2 $BoundsMin :Operation Subtract :OutputTarget Size))
  (bind ExtractAttribute_0 (Extract%20Attribute%20at%20Index :bForceOutputAttributeToBeInElementsDomain true :InputSource Size))
  (bind Subgraph_9 (Subgraph :Graph /PCGPrimitives/Primitives/Extract/Extract_Segments.Extract_Segments))
  (bind Subgraph_8 (Subgraph :Graph /PCGPrimitives/Primitives/Assign/Assign_Rule.Assign_Rule))
  (connect CombinePoints_0 Out AttributeMathsOp_0 InA)
  (connect CombinePoints_0 Out AttributeMathsOp_0 InB)
  (connect GetSplineControlPoints_0 Out CombinePoints_0 In)
  (connect AttributeMathsOp_0 Out ExtractAttribute_0 In)
  (connect Subgraph_3 Out FilterDataByType_2 In)
  (connect FilterDataByType_2 InsideFilter GetSplineControlPoints_0 In)
  (connect PlaceOnGrid Out Subgraph_1 PrimaryInput)
  (connect Subgraph_1 Out Subgraph_3 PrimaryInput)
  (connect Subgraph_9 Out Subgraph_8 PrimaryInput)
  (connect Subgraph_3 Out Subgraph_9 PrimaryInput))
