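; PCG example graph that scatters assets over the surfaces of input mesh actors through the Spawn_Assets subgraph.
; Each actor's scale and largest bounds dimension convert the sampling radius and percent-of-size minimum into mesh-local values.
; A loop Poisson-samples each static mesh using the larger radius, capped by point count.
; Samples take the actor transform but keep their original scale; the random range parameters are unconnected.
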(graph Scatter_points_onto_Mesh_Actor
  (params (Mesh_Actor :type (Array (SoftObject /Script/Engine.Actor))) (Sampling_Radius :type Double :default 5.0) (Percent_Size_limit_for_sampling :type Double :default 0.01) (Max_Points_Per_Mesh :type Int :default 5000) (Assets_to_scatter_on_the_Mesh_Actors :type (Array (SoftObject /Script/CoreUObject.Object)) :default [/PCG/DebugObjects/PCG_Cube.PCG_Cube]) (Random_Range_Postion :type Vector) (Random_Range_Rotation :type Vector) (Random_Range_Scale :type Vector) (Position_Offset :type Vector) (Rotation__Offset :type Rotator) (Scale_offset :type Vector :default (1.0 1.0 1.0)) (Uniform_Scale :type Bool :default true))
  (bind GetActorProperty_0 (Get%20Actor%20Property :ActorSelector (struct :ActorFilter FromInput) :OutputAttributeName Mesh :PropertyName StaticMeshComponent.staticMesh :In Mesh_Actor))
  (bind GetActorData_0 (Get%20Actor%20Data :ActorSelector (struct :ActorFilter FromInput) :Mode GetSinglePoint :In Mesh_Actor))
  (bind ExtractAttribute_0 (Extract%20Attribute%20at%20Index :bForceOutputAttributeToBeInElementsDomain true :InputSource $Transform :OutputAttributeName trans))
  (bind CopyAttributes_1 (Copy%20Attributes :InputSource MaxSize :OutputTarget MaxSize))
  (bind CopyAttributes_5 (Copy%20Attributes :InputSource Sampling_Radius :OutputTarget Sampling_Radius))
  (bind Loop_1 (Loop :Graph Mesh_Sampling_loop :Max_Points_Per_Mesh Max_Points_Per_Mesh :Sampling_Radius Sampling_Radius))
  (bind AttributeTransformOp_0 (Attribute%20Transform%20Op :InputSource1 $Transform :InputSource2 Trans))
  (bind CopyAttributes_0 (Copy%20Attributes :InputSource $Scale :OutputTarget $Scale))
  (bind Subgraph_1 (Subgraph :Graph /PCGPrimitives/Primitives/Spawn/Spawn_Assets.Spawn_Assets :Assets Assets_to_scatter_on_the_Mesh_Actors :PositionOffset Position_Offset :RotationOffset Rotation__Offset :ScaleOffset Scale_offset :UniformScale Uniform_Scale))
  (bind WriteDataIndex_0 (Write%20Data%20Index :bAddAttribute true :bAddTag false :IndexAttribute dataIndex))
  ;[0.15 0.15 0.15 0.5] Sampling radius function of mesh actor scale
  (bind CopyAttributes_2 (Copy%20Attributes :InputSource $Scale.X :OutputTarget scale))
  (bind CopyAttributes_3 (Copy%20Attributes :InputSource $Scale.Y :OutputTarget scale))
  (bind CopyAttributes_4 (Copy%20Attributes :InputSource $Scale.Z :OutputTarget scale))
  (bind MergePoints_1 (Merge%20Points))
  (bind AttributePartition_1 (Attribute%20Partition :PartitionAttributeSelectors [dataIndex]))
  (bind AttributeReduce_2 (Attribute%20Reduce :InputSource scale))
  (bind AttributeMathsOp_6 (Attribute%20Maths%20Op :InputSource1 Sampling_Radius :Operation Divide :InA Sampling_Radius))
  ;[0.15 0.15 0.15 0.5] Measuring max size of object to set percentage limit for radius search
  (bind ApplyScaleToBounds_0 (Apply%20Scale%20To%20Bounds))
  (bind AttributeMathsOp_0 (Attribute%20Maths%20Op :InputSource1 $BoundsMax.X :InputSource2 $BoundsMin.X :Operation Subtract :OutputTarget Size))
  (bind AttributeMathsOp_1 (Attribute%20Maths%20Op :InputSource1 $BoundsMax.Y :InputSource2 $BoundsMin.Y :Operation Subtract :OutputTarget Size))
  (bind AttributeMathsOp_2 (Attribute%20Maths%20Op :InputSource1 $BoundsMax.Z :InputSource2 $BoundsMin.Z :Operation Subtract :OutputTarget Size))
  (bind MergePoints_0 (Merge%20Points))
  (bind AttributePartition_0 (Attribute%20Partition :PartitionAttributeSelectors [dataIndex]))
  (bind AttributeReduce_0 (Attribute%20Reduce :Operation Max :OutputAttributeName MaxSize))
  (bind AttributeMathsOp_5 (Attribute%20Maths%20Op :InputSource1 MaxSize :Operation Multiply :InB Percent_Size_limit_for_sampling))
  (bind AttributeMathsOp_7 (Attribute%20Maths%20Op :InputSource1 MaxSize :Operation Divide))
  (reroute Mesh_actors_bounds WriteDataIndex_0)
  (connect Mesh_actors_bounds Out ApplyScaleToBounds_0 In)
  (connect ApplyScaleToBounds_0 Out AttributeMathsOp_0 InA)
  (connect ApplyScaleToBounds_0 Out AttributeMathsOp_0 InB)
  (connect ApplyScaleToBounds_0 Out AttributeMathsOp_1 InA)
  (connect ApplyScaleToBounds_0 Out AttributeMathsOp_1 InB)
  (connect ApplyScaleToBounds_0 Out AttributeMathsOp_2 InA)
  (connect ApplyScaleToBounds_0 Out AttributeMathsOp_2 InB)
  (connect AttributeReduce_0 Out AttributeMathsOp_5 InA)
  (connect AttributeReduce_2 Out AttributeMathsOp_6 InB)
  (connect AttributeMathsOp_5 Out AttributeMathsOp_7 InA)
  (connect AttributeReduce_2 Out AttributeMathsOp_7 InB)
  (connect MergePoints_0 Out AttributePartition_0 In)
  (connect MergePoints_1 Out AttributePartition_1 In)
  (connect AttributePartition_0 Out AttributeReduce_0 In)
  (connect AttributePartition_1 Out AttributeReduce_2 In)
  (connect Loop_1 Mesh%20Sampler%20Out AttributeTransformOp_0 InA)
  (connect ExtractAttribute_0 Out AttributeTransformOp_0 InB)
  (connect Loop_1 Mesh%20Sampler%20Out CopyAttributes_0 Source)
  (connect AttributeTransformOp_0 Out CopyAttributes_0 Target)
  (connect AttributeMathsOp_7 Out CopyAttributes_1 Source)
  (connect GetActorProperty_0 Out CopyAttributes_1 Target)
  (connect Mesh_actors_bounds Out CopyAttributes_2 Source)
  (connect Mesh_actors_bounds Out CopyAttributes_2 Target)
  (connect Mesh_actors_bounds Out CopyAttributes_3 Source)
  (connect Mesh_actors_bounds Out CopyAttributes_3 Target)
  (connect Mesh_actors_bounds Out CopyAttributes_4 Source)
  (connect Mesh_actors_bounds Out CopyAttributes_4 Target)
  (connect AttributeMathsOp_6 Out CopyAttributes_5 Source)
  (connect CopyAttributes_1 Out CopyAttributes_5 Target)
  (connect Mesh_actors_bounds Out ExtractAttribute_0 In)
  (connect CopyAttributes_5 Out Loop_1 In)
  (connect AttributeMathsOp_0 Out MergePoints_0 In)
  (connect AttributeMathsOp_1 Out MergePoints_0 In)
  (connect AttributeMathsOp_2 Out MergePoints_0 In)
  (connect CopyAttributes_2 Out MergePoints_1 In)
  (connect CopyAttributes_3 Out MergePoints_1 In)
  (connect CopyAttributes_4 Out MergePoints_1 In)
  (connect CopyAttributes_0 Out Subgraph_1 PrimaryInput)
  (connect GetActorData_0 Out WriteDataIndex_0 In)
  (output Out Subgraph_1))

(subgraph Mesh_Sampling_loop
  (params (Sampling_Radius :type Double :default 5.0) (Max_Points_Per_Mesh :type Int :default 5000))
  (inputs (Multiply%20Out :type Param))
  (outputs (Mesh%20Sampler%20Out :type Point))
  (bind ExtractAttribute_0 (Extract%20Attribute%20at%20Index :bForceOutputAttributeToBeInElementsDomain true :InputSource MaxSize))
  (bind ExtractAttribute_1 (Extract%20Attribute%20at%20Index :bForceOutputAttributeToBeInElementsDomain true :InputSource Sampling_Radius))
  (bind AttributeMathsOp_0 (Attribute%20Maths%20Op :Operation Max))
  (bind MeshSampler_0 (Mesh%20Sampler :bExtractUVAsAttribute true :bHasDefaultColorOverride false :bOutputTriangleIds true :SamplingMethod PoissonSampling :SamplingOptions (struct :SamplingRadius 2.491835 :MaxNumSamples 5000) :VoxelSize 5.0 :MaxNumSamples Max_Points_Per_Mesh))
  (connect ExtractAttribute_1 Out AttributeMathsOp_0 InA)
  (connect ExtractAttribute_0 Out AttributeMathsOp_0 InB)
  (connect DefaultInputNode In ExtractAttribute_0 In)
  (connect DefaultInputNode In ExtractAttribute_1 In)
  (connect DefaultInputNode In MeshSampler_0 Mesh)
  (connect AttributeMathsOp_0 Out MeshSampler_0 SamplingRadius)
  (output Mesh%20Sampler%20Out MeshSampler_0))
